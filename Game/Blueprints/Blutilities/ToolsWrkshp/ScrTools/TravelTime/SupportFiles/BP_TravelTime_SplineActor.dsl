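; BP_TravelTime_SplineActor is a support actor for the TravelTime editor scripting tool, named as a spline actor.
; Its only logic is WaitForTime, a polled throttle that compares CurTime against LastWaitTime plus a pause interval.
; When the interval has passed it stores CurTime as LastWaitTime and returns true, otherwise false.
; BeginPlay, Tick, overlap and construction script are empty, and nothing here updates CurTime.
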(event EventBeginPlay)

(event Collision|EventActorBeginOverlap
  (params (OtherActor :type Actor%20Object%20Reference)))

(event EventTick
  (params (DeltaSeconds :type Float%20%28single-precision%29)))

(fn ConstructionScript)

(fn WaitForTime
  (params (PauseTime :type Float%20%28double-precision%29 :default 0.5))
  (if (> (Variables|Default|GetCurTime) (+ PauseTime (Variables|Default|GetLastWaitTime)))
    (Variables|Default|SetLastWaitTime (Variables|Default|GetCurTime))
    (return true)
    (else
      (return false))))
