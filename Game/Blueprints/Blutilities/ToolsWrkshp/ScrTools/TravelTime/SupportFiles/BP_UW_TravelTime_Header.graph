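# Header widget for the TravelTime tool, showing distance and time readouts plus a reference list of segment-editing shortcuts.
# PreConstruct sets up two shared expandable boxes, "Tool Information" and "Shortcuts", and caches their dynamic-entry containers.
# The first box gets Total Distance and Total Time lines initialised to zero; the second fills from the Shortcut Map variable.
# Created lines are stored and passed back into the setup calls; nothing here updates the values.
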WidgetBlueprint /Game/Blueprints/Blutilities/ToolsWrkshp/ScrTools/TravelTime/SupportFiles/BP_UW_TravelTime_Header
graphs [EventGraph {nodes [
   Event PreConstruct()
   > n1 = Sequence()
   n2 = Get ReadoutBox() -> (ReadoutBox:BP UW Expandable Box Object Reference)
   > SetUpBox(self=@n2, Name="LOCTEXT(\"Tool Information\")")
   > Set Readout Dynamic Entries(Readout Dynamic Entries:Vertical Box Object Reference=Get DynamicEntries(self=@n2)) -> (Output_Get:Vertical Box Object Reference)
   > n3 = Sequence()
   > n4 = SetupLine!(
    execute
    self:BP UW Expandable Box Object Reference=Get ReadoutBox()
    CurrentLine:BP UW Readout Line Object Reference=Get DistanceLine() -> (DistanceLine:BP UW Readout Line Object Reference)
    DynamicLine:Vertical Box Object Reference=Get Readout Dynamic Entries() -> (Readout Dynamic Entries:Vertical Box Object Reference)
    Label:Text="LOCTEXT(\"Total Distance:\")"
    Value:Text="LOCTEXT(\"0\")"
   ) -> (then, Readout Line:BP UW Readout Line Object Reference)
   > Set DistanceLine(DistanceLine:BP UW Readout Line Object Reference=@n4) -> (Output_Get:BP UW Readout Line Object Reference)
   n5 = SetupLine(
    execute=@n3.then_1
    self=Get ReadoutBox()
    CurrentLine=Get TimeLine() -> (TimeLine:BP UW Readout Line Object Reference)
    DynamicLine=Get Readout Dynamic Entries()
    Label="LOCTEXT(\"Total Time:\")"
    Value="LOCTEXT(\"0\")"
   )
   > Set TimeLine(TimeLine:BP UW Readout Line Object Reference=@n5) -> (Output_Get:BP UW Readout Line Object Reference)
   n6 = Get ShortcutBox() -> (ShortcutBox:BP UW Expandable Box Object Reference)
   SetUpBox(execute=@n1.then_1, self=@n6, Name="LOCTEXT(\"Shortcuts\")")
   > Set Shortcut Dynamic Entries(Shortcut Dynamic Entries:Vertical Box Object Reference=Get DynamicEntries(self=@n6)) -> (Output_Get:Vertical Box Object Reference)
   > n7 = AddShortcuts!(
    execute
    self:BP UW Expandable Box Object Reference=Get ShortcutBox()
    ShortcutMap:Map of Strings to Strings=Get Shortcut Map() -> (Shortcut Map:Map of Strings to Strings)
    ExistingShortcutLines:Array of BP UW Readout Line Object References=Get Existing Shortcut Lines() -> (Existing Shortcut Lines:Array of BP UW Readout Line Object References)
    ShortcutMapKeys:Array of Strings
   ) -> (then, AllShortcutLines:Array of BP UW Readout Line Object References)
   > Set Existing Shortcut Lines(Existing Shortcut Lines:Array of BP UW Readout Line Object References=@n7) -> (Output_Get:Array of BP UW Readout Line Object References)
   Event Construct()
   Event Tick()
  ]}]
variables [
 DistanceLine {type=BP_UW_ReadoutLine_C, default=null}
 Existing Shortcut Lines {type="array<BP_UW_ReadoutLine_C>", default []}
 Readout Dynamic Entries {type=VerticalBox, default=null}
 Shortcut Dynamic Entries {type=VerticalBox, default=null}
 Shortcut Map {
  type="map<string>"
  default {
   Add Segment To Beginning=Hold B + Click
   Delete Segment=Hold V + Click on Segment
   Divide Segment=Hold C + Click on Segment
   Move Segment=Hold X
  }
 }
 TimeLine {type=BP_UW_ReadoutLine_C, default=null}
]
widgets {
 info {inheritedWidgetCount=0, parentClass=&UserWidget, rootWidgetClass=&Border, widgetCount=4}
 widgets [
  Border_51 {
   widgetClassPath=&Border
   widget=Border_51
   properties {
    background {
     bIsDynamicallyLoaded=false
     drawAs=Image
     imageType=FullColor
     mirroring=NoMirror
     resourceName=None
     resourceObject=null
     tiling=NoTile
     imageSize=<256 256>
     margin=<ltrb 0 0 0 0>
     outlineSettings {
      bUseBrushTransparency=false
      roundingType=HalfHeightRadius
      width=0
      color {colorUseRule=UseColor_Specified, specifiedColor=<rgba 0 0 0 0>}
      cornerRadii=<0 0 0 0>
     }
     tintColor {colorUseRule=UseColor_Specified, specifiedColor=<rgba 0.0469 0.0469 0.0469 0>}
     uVRegion {bIsValid=false, max=<0 0>, min=<0 0>}
    }
    brushColor=<rgba 0 0 0 1>
    contentColorAndOpacity=<rgba 0.9473 0.9473 0.9473 1>
   }
  }
  VerticalBox_0 {widgetClassPath=&VerticalBox, parent=Border_51, widget=VerticalBox_0, slot {class=BorderSlot}}
  ReadoutBox {
   widgetClassPath=/Game/Blueprints/Blutilities/ToolsWrkshp/ScrTools/Shared/SharedWidgets/ExpandableBox/BP_UW_ExpandableBox
   bIsVariable=true
   parent=VerticalBox_0
   widget=ReadoutBox
   properties {boxColor=<rgba 0 0 0 0>}
   slot {class=VerticalBoxSlot}
  }
  ShortcutBox {
   widgetClassPath=&BP_UW_ExpandableBox
   bIsVariable=true
   parent=VerticalBox_0
   widget=ShortcutBox
   properties {boxColor=<rgba 0 0 0 0>}
   slot {class=VerticalBoxSlot}
  }
 ]
}
