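# Editor-only utility Blueprint in the TravelTime tool support files that reads or sets Game View on the active level viewport.
# Its one function, GetSetGameView, gets the active viewport config key from LevelEditorSubsystem, then branches on the SetGameView input to either apply GameViewVal or query the current state.
# It returns the key and IsGameView, which is wired only from the query, so the set path does not refresh it.
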EditorUtilityBlueprint /Game/Blueprints/Blutilities/ToolsWrkshp/ScrTools/TravelTime/SupportFiles/BP_TravelTime_EditorOnlyUtils
graphs [
 EventGraph {links [], nodes []}
 GetSetGameView {nodes [
   n1 = GetSetGameView!() -> (then, SetGameView:bool=~, GameViewVal:bool=~) {is=title K2Node_FunctionEntry}
   n2 = Get LevelEditorSubsystem()
   > n3 = GetActiveViewportConfigKey!(execute, self:Level Editor Subsystem Object Reference=@n2) -> (then, ReturnValue:Name=None)
   > Set ConfigKey(ConfigKey:Name=@n3) -> (Output_Get:Name=None)
   > n4 = Branch(Condition=@n1)
   > n5 = EditorSetGameView!(
    execute
    self:Level Editor Subsystem Object Reference=@n2
    bGameView:bool=@n1.GameViewVal
    ViewportConfigKey:Name=Get ConfigKey() -> (ConfigKey:Name=None)
   ) -> (then)
   n6 = EditorGetGameView!(execute=@n4.else, self:Level Editor Subsystem Object Reference=@n2, ViewportConfigKey:Name=Get ConfigKey() -> (ConfigKey=None)) -> (then, ReturnValue:bool)
   Return Node!(execute=[@n6.then, @n5.then], ViewportKey:Name=Get ConfigKey() -> (ConfigKey=None), IsGameView:bool=@n6)
  ]}
]
variables []
